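# Physics asset for the Banner Flag cosmetic prop, giving its pole and cloth bones simplified bodies for physics-driven flag motion.
# A capsule covers flag_pole, and three thin boxes cover the dyn_flag bones.
# Constraints chain the segments outward from the pole, each sharing the same angular limits with a wider first swing axis, approximating cloth with linked rigid panels.
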PhysicsAsset /BRCosmetics/Animation/Props/BannerFlag/Meshes/Banner_Flag_Physics
bodies {
 dyn_flag_01 [{shapeType=Box, extentX=32.5, extentY=0.5022, extentZ=25.166, center=<16.8101 0 0>, rotation=<rot 90 0 0>}]
 dyn_flag_02 [{shapeType=Box, extentX=32.5, extentY=0.5022, extentZ=25.166, center=<16.8101 0 0>, rotation=<rot 90 0 0>}]
 dyn_flag_03 [{shapeType=Box, extentX=32.5, extentY=0.5022, extentZ=25.166, center=<11.2082 0 0>, rotation=<rot 90 0 0>}]
 flag_pole [{shapeType=Capsule, length=171.1559, radius=1.9675, center=<0 -0.033 55.0795>}]
}
constraints [
 {bone1Name=dyn_flag_01, bone2Name=flag_pole, swing1LimitDegrees=85, swing2LimitDegrees=45, twistLimitDegrees=45}
 {bone1Name=dyn_flag_02, bone2Name=dyn_flag_01, swing1LimitDegrees=85, swing2LimitDegrees=45, twistLimitDegrees=45}
 {bone1Name=dyn_flag_03, bone2Name=dyn_flag_02, swing1LimitDegrees=85, swing2LimitDegrees=45, twistLimitDegrees=45}
]
