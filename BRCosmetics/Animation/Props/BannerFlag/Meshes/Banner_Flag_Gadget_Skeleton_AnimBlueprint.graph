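# Animation Blueprint for the banner flag gadget prop in BR cosmetics, adding procedural cloth-like motion to the flag skeleton.
# The input pose runs through a RigidBody node using the banner flag physics asset, with overridden gravity and scaled, clamped component motion.
# A Trail controller on a three-bone chain ending at dyn_flag_03 then adds lag before output.
# The event graph holds only unconnected nodes, so it drives no logic.
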AnimBlueprint /BRCosmetics/Animation/Props/BannerFlag/Meshes/Banner_Flag_Gadget_Skeleton_AnimBlueprint
target_skeleton=&/Banner_Flag_Gadget_Skeleton
graphs [
 AnimGraph {nodes [
   n1 = RigidBody(ComponentPose=Local To Component(LocalPose=Input Pose())) {
    bClampLinearTranslationLimitToRefPose=true
    bOverrideWorldGravity=true
    overridePhysicsAsset=&/Banner_Flag_Physics
    alphaBoolBlend {blendOption=Linear}
    componentAppliedLinearAccClamp=<2000 2000 2000>
    componentLinearAccScale=<0.8 0.8 0.8>
    componentLinearVelScale=<0.8 0.8 0.8>
    overrideWorldGravity {z=-980}
   }
   n2 = Trail controller - Bone: dyn_flag_03(ComponentPose=@n1, Alpha=1) {
    chainLength=3
    lODThreshold=1
    relaxationSpeedScale=2
    alphaScaleBiasClamp {bInterpResult=true, interpSpeedDecreasing=4, interpSpeedIncreasing=4}
    rotationLimits [{limitMax=<180 180 180>}, {limitMax=<180 180 180>}, {limitMax=<180 180 180>}]
    rotationOffsets [<0 0 0>, <0 0 0>, <0 0 0>]
    trailBone {boneName=dyn_flag_03}
   }
   Output Pose(Result=Component To Local(ComponentPose=@n2)) {layerGroup=None}
  ]}
 EventGraph {nodes [
   TryGetPawnOwner()
   GetPartSkeletalMeshComponent()
   Event BlueprintUpdateAnimation()
  ]}
]
variables []
